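; Glider_Inspect is an Art QA Blueprint with a box volume that puts overlapping Fort player pawns into a vortex flight state, apparently for glider inspection.
; Begin overlap calls FlyOn to enable vortex and apply a gameplay effect (commented as Fly); end overlap calls FlyOff to undo both.
; Both custom events read OtherActor directly from their overlap event.
; ToggleGliderInspection flips an enabled flag that drives the box's overlap event generation.
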(event Custom|ToggleGliderInspection
  ;[1 1 1] Allow certain configurations to opt out of this behavior at runtime
  (bind _output_get (Variables|Default|SetbGliderInspectEnabled (not (Variables|Default|GetbGliderInspectEnabled))))
  (Variables|/Script/Engine.PrimitiveComponent|SetbGenerateOverlapEvents _output_get (Variables|Default|GetBox)))

(event OnComponentBeginOverlap(Box)
  (params (OverlappedComponent :type Primitive%20Component%20Object%20Reference) (OtherActor :type Actor%20Object%20Reference) (OtherComp :type Primitive%20Component%20Object%20Reference) (OtherBodyIndex :type Integer) (bFromSweep :type Boolean) (SweepResult :type Hit%20Result%20Structure%20%28by%20ref%29))
  ;[1 1 1] Fly GE applied
  (CallFunction|FlyOn))

(event Custom|FlyOn
  ;[1 1 1] Fly GE applied
  (bind _asfort_player_pawn (Utilities|Casting|CastToFortPlayerPawn (stale OnComponentBeginOverlap(Box) OtherActor)))
  (Vortex|SetinVortex _asfort_player_pawn true 5e+02 8e+02 2e+03 -0.25 5e+03 true)
  (GameplayEffects|ApplyGameplayEffectToTarget :self (Ability|GetAbilitySystemComponent _asfort_player_pawn) :Target (Ability|GetAbilitySystemComponent _asfort_player_pawn) :Level 0.0))

(event OnComponentEndOverlap(Box)
  (params (OverlappedComponent :type Primitive%20Component%20Object%20Reference) (OtherActor :type Actor%20Object%20Reference) (OtherComp :type Primitive%20Component%20Object%20Reference) (OtherBodyIndex :type Integer))
  ;[1 1 1] Fly GE applied
  (CallFunction|FlyOff))

(event Custom|FlyOff
  ;[1 1 1] Fly GE applied
  (bind _asfort_player_pawn (Utilities|Casting|CastToFortPlayerPawn (stale OnComponentEndOverlap(Box) OtherActor)))
  (Vortex|SetinVortex _asfort_player_pawn false 5e+02 8e+02 2e+03 -0.25 5e+03 true)
  (GameplayEffects|RemoveActiveGameplayEffectbySourceEffect (Ability|GetAbilitySystemComponent _asfort_player_pawn) 0 (Ability|GetAbilitySystemComponent _asfort_player_pawn) -1))

(fn ConstructionScript)
